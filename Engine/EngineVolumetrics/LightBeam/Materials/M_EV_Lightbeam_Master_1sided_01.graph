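# Unlit translucent, one-sided master material for volumetric-style light beams from Engine Volumetrics, with parameters for color, opacity and fading.
# It fakes soft beam edges by turning the tangent-space reflection vector into sphere-map-like UVs for two multiplied falloff textures.
# An optional panning dust texture modulates that falloff, which is then scaled by opacity and a near-camera fade.
# Static switches toggle the dust overlay, a far-distance fade and depth fade against scene geometry.
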Material /Engine/EngineVolumetrics/LightBeam/Materials/M_EV_Lightbeam_Master_1sided_01
graph {
 properties {bUsedWithStaticLighting=true, blendMode=BLEND_Translucent, pixelDepthOffsetMode=PDOM_Legacy, shadingModel=MSM_Unlit, usedShadingModels=Unlit}
 nodes [
  Constant() {r=0.56}
  Multiply(
   A=Divide(A=Constant() {r=0.0001}, B=ScalarParameter() {defaultValue=1, parameterName=FarFadeDistance})
   B=Subtract(A=PixelDepth().R, B=ScalarParameter() {defaultValue=1024, parameterName=FarFadeOffset})
  )
  n1 = VectorParameter() {desc="Color (A is luminosity)", parameterName=Color, defaultValue=<rgba 1 1 1 1>}
  Multiply(A=@n1.RGB, B=@n1.RGB)
  n2 = ScalarParameter() {defaultValue=1, parameterName=TextureScaleX}
  n3 = ScalarParameter() {defaultValue=1, parameterName=TextureScaleY}
  n4 = Panner(
   Coordinate=Multiply(A=TextureCoordinate(), B=AppendVector(A=@n2, B=@n3))
   Time=Multiply(A=Time(), B=ScalarParameter() {defaultValue=1, parameterName=PanningSpeed})
  ) {speedX=0.01, speedY=0.04}
  n5 = Transform(None=ReflectionVectorWS()) {transformSourceType=TRANSFORMSOURCE_World, transformType=TRANSFORM_Tangent}
  n6 = ComponentMask(None=@n5) {b=true}
  n7 = Constant() {r=8}
  n8 = SquareRoot(None=Multiply(A=Add(A=@n6, B=@n7), B=Constant() {r=0.8})) {desc=Z Component Square}
  n9 = Constant() {r=0.5}
  n10 = Add(A=Divide(A=ComponentMask(None=@n5) {desc=X Component, r=true}, B=@n8), B=@n9)
  n11 = ComponentMask(None=TextureCoordinate()) {g=true}
  n12 = Add(A=@n9, B=Divide(A=ComponentMask(None=@n5) {desc=Y Component, g=true}, B=@n8))
  n13 = Multiply(
   A=TextureSampleParameter2D(UVs=AppendVector(A=@n10, B=@n11)) {parameterName=FalloffTexture, texture=&/T_EV_LightBeam_Falloff_01}.RGB
   B=TextureSample(UVs=AppendVector(A=@n10, B=@n12)) {samplerType=SAMPLERTYPE_Grayscale, texture=&/T_EV_LightBeam_Falloff_02}.RGB
  )
  n14 = Multiply(
   A=LinearInterpolate(
    A=Constant() {r=1}
    B=TextureSampleParameter2D(UVs=@n4) {parameterName=MultiplyTexture1, texture=&/T_EV_DustPanner_01}.RGB
    Alpha=ScalarParameter() {defaultValue=0.8, parameterName=TextureContrast}
   )
   B=@n13
  )
  n15 = Multiply(A=Divide(A=Constant() {r=0.0025}, B=ScalarParameter() {defaultValue=1, parameterName=NearFadeDistance}), B=PixelDepth().R)
  n16 = Constant() {r=1}
  n17 = OneMinus()
  n18 = Constant() {r=1}
  n19 = Multiply(
   A=Multiply(
    A=StaticSwitchParameter(True=@n14, False=@n13) {parameterName=UseTextureOverlay}
    B=ScalarParameter() {defaultValue=0.25, parameterName=Opacity}
   )
   B=Multiply(A=Clamp(None=@n15, Max=@n16), B=StaticSwitchParameter(True=@n17, False=@n18) {parameterName=Fade when Far})
  )
  StaticSwitchParameter(True=DepthFade(Opacity=@n19, FadeDistance=ScalarParameter() {defaultValue=100, desc=Density, parameterName=DensityDistance}), False=@n19) {parameterName=UseDepthFade}
 ]
}
